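# Physics asset for the PollenTrove female medium cosmetic's bandolier, providing bodies for its simulated straps and bag.
# Front, rear and central bag-sim capsule chains each hang from a master body locked to spine_05, with tight angular limits between segments.
# Each chain ends in a locked long capsule, and the front and rear ends lock to the central chain's end, closing the loop.
# Pelvis, spine and right thigh capsules supply the body shapes.
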PhysicsAsset /BRCosmetics/Characters/Player/Female/Medium/Bodies/F_MED_PollenTrove/Meshes/F_MED_PollenTrove_Bandolier_B
bodies {
 def_bandolier_front_01 [{shapeType=Capsule, length=7.911, radius=3, shapeName=def_bag_sim_01_capsule, center=<4.3216 -2.2917 0>, rotation=<rot 90 -2.8129 0>}]
 def_bandolier_front_01_MSTR [{shapeType=Capsule, length=0.5, radius=2, shapeName=dyn_bag_sim_chain_01_MSTR_capsule}]
 def_bandolier_front_02 [{shapeType=Capsule, length=8.318, radius=3, shapeName=def_bag_sim_02_capsule, center=<4.159 -1.373 0>, rotation=<rot 90 0 0>}]
 def_bandolier_front_03 [{shapeType=Capsule, length=7.547, radius=3, shapeName=def_bag_sim_03_capsule, center=<4.1949 -2.1185 0>, rotation=<rot 90 11.2499 0>}]
 def_bandolier_front_04 [{shapeType=Capsule, length=6.856, radius=3, shapeName=def_bag_sim_04_capsule, center=<3.5224 -0.1151 0>, rotation=<rot 90 39.3748 0>}]
 def_bandolier_front_05 [{shapeType=Capsule, length=40, radius=3.25, shapeName=def_bag_sim_05_capsule, center=<0 9.5148 0>, rotation=<rot 90 90 0>}]
 def_bandolier_rear_01 [{shapeType=Capsule, length=7.911, radius=3, shapeName=def_bag_sim_01_capsule, center=<3.7599 2.3112 0>, rotation=<rot 90 8.4374 0>}]
 def_bandolier_rear_01_MSTR [{shapeType=Capsule, length=0.5, radius=2, shapeName=dyn_bag_sim_chain_01_MSTR_capsule}]
 def_bandolier_rear_02 [{shapeType=Capsule, length=8.318, radius=3.6, shapeName=def_bag_sim_02_capsule, center=<4.1087 1.5058 0>, rotation=<rot 90 -2.8128 0>}]
 def_bandolier_rear_03 [{shapeType=Capsule, length=7.547, radius=3.6, shapeName=def_bag_sim_03_capsule, center=<3.9446 1.2811 0>, rotation=<rot 90 -11.2501 0>}]
 def_bandolier_rear_04 [{shapeType=Capsule, length=6.856, radius=3.6, shapeName=def_bag_sim_04_capsule, center=<3.3141 -0.1616 0>, rotation=<rot 90 -33.7504 0>}]
 def_bandolier_rear_05 [{shapeType=Capsule, length=40, radius=3.25, shapeName=def_bag_sim_05_capsule, center=<0 -6.7619 0>, rotation=<rot 90 90 0>}]
 dyn_bag_sim_chain_01 [{shapeType=Capsule, length=7.911, radius=2, shapeName=def_bag_sim_01_capsule, center=<7.911 0 0>, rotation=<rot 90 0 0>}]
 dyn_bag_sim_chain_01_MSTR [{shapeType=Capsule, length=0.5, radius=2, shapeName=dyn_bag_sim_chain_01_MSTR_capsule}]
 dyn_bag_sim_chain_02 [{shapeType=Capsule, length=8.318, radius=2, shapeName=def_bag_sim_02_capsule, center=<4.159 0 0>, rotation=<rot 90 0 0>}]
 dyn_bag_sim_chain_03 [{shapeType=Capsule, length=7.547, radius=2, shapeName=def_bag_sim_03_capsule, center=<3.7735 0 0>, rotation=<rot 90 0 0>}]
 dyn_bag_sim_chain_04 [{shapeType=Capsule, length=6.856, radius=2, shapeName=def_bag_sim_04_capsule, center=<3.428 0 0>, rotation=<rot 90 0 0>}]
 dyn_bag_sim_chain_05 [{shapeType=Capsule, length=40, radius=3.25, shapeName=def_bag_sim_05_capsule, center=<0 1.6435 0>, rotation=<rot 90 90 0>}]
 pelvis [{shapeType=Capsule, length=10, radius=12, shapeName=pelvis_capsule, center=<-2.7367 -1.0918 3.8442>, rotation=<rot 90 -0.0001 0>}]
 spine_01 [{shapeType=Capsule, length=7.7158, radius=12.3638, center=<3.3 0 0>}]
 spine_02 [
  {shapeType=Capsule, length=11.491, radius=10.7317, center=<3.2455 0.1555 0>}
  {shapeType=Capsule, length=20.6875, radius=9.3125, center=<-1.8317 1.9362 0>, rotation=<rot 90 0 0>}
 ]
 spine_03 [{shapeType=Capsule, length=10.1863, radius=10.9854, center=<3.48 0.7183 0>}]
 spine_04 [{shapeType=Capsule, length=13.0076, radius=9, center=<3.4995 -0.029 0>}]
 spine_05 [{shapeType=Capsule, length=20.678, radius=7, center=<0.0328 -0.8704 0>}]
 thigh_r [{shapeType=Capsule, length=29, radius=9, shapeName=thigh_r_capsule, center=<14.25 0 0>, rotation=<rot 90 0 0>}]
}
constraints [
 {
  bone1Name=spine_01
  bone2Name=pelvis
  swing1LimitDegrees=45
  swing1Motion=Free
  swing2LimitDegrees=45
  swing2Motion=Free
  twistLimitDegrees=45
  twistMotion=Free
 }
 {
  bone1Name=spine_02
  bone2Name=spine_01
  swing1LimitDegrees=45
  swing1Motion=Free
  swing2LimitDegrees=45
  swing2Motion=Free
  twistLimitDegrees=45
  twistMotion=Free
 }
 {
  bone1Name=spine_03
  bone2Name=spine_02
  swing1LimitDegrees=45
  swing1Motion=Free
  swing2LimitDegrees=45
  swing2Motion=Free
  twistLimitDegrees=45
  twistMotion=Free
 }
 {
  bone1Name=spine_04
  bone2Name=spine_03
  swing1LimitDegrees=45
  swing1Motion=Free
  swing2LimitDegrees=45
  swing2Motion=Free
  twistLimitDegrees=45
  twistMotion=Free
 }
 {
  bone1Name=spine_05
  bone2Name=spine_04
  swing1LimitDegrees=45
  swing1Motion=Free
  swing2LimitDegrees=45
  swing2Motion=Free
  twistLimitDegrees=45
  twistMotion=Free
 }
 {
  bone1Name=thigh_r
  bone2Name=pelvis
  swing1LimitDegrees=45
  swing1Motion=Free
  swing2LimitDegrees=45
  swing2Motion=Free
  twistLimitDegrees=45
  twistMotion=Free
 }
 {
  bone1Name=dyn_bag_sim_chain_01_MSTR
  bone2Name=spine_05
  swing1LimitDegrees=45
  swing1Motion=Locked
  swing2LimitDegrees=45
  swing2Motion=Locked
  twistLimitDegrees=45
  twistMotion=Locked
 }
 {bone1Name=dyn_bag_sim_chain_01, bone2Name=dyn_bag_sim_chain_01_MSTR, swing1LimitDegrees=5.625, swing2LimitDegrees=5.625, twistLimitDegrees=0}
 {bone1Name=dyn_bag_sim_chain_02, bone2Name=dyn_bag_sim_chain_01, swing1LimitDegrees=5.625, swing2LimitDegrees=5.625, twistLimitDegrees=0}
 {bone1Name=dyn_bag_sim_chain_03, bone2Name=dyn_bag_sim_chain_02, swing1LimitDegrees=5.625, swing2LimitDegrees=5.625, twistLimitDegrees=0}
 {bone1Name=dyn_bag_sim_chain_04, bone2Name=dyn_bag_sim_chain_03, swing1LimitDegrees=5.625, swing2LimitDegrees=5.625, twistLimitDegrees=0}
 {
  bone1Name=dyn_bag_sim_chain_05
  bone2Name=dyn_bag_sim_chain_04
  swing1LimitDegrees=0
  swing1Motion=Locked
  swing2LimitDegrees=5.625
  swing2Motion=Locked
  twistLimitDegrees=0
  twistMotion=Locked
 }
 {
  bone1Name=def_bandolier_front_01_MSTR
  bone2Name=spine_05
  swing1LimitDegrees=45
  swing1Motion=Locked
  swing2LimitDegrees=45
  swing2Motion=Locked
  twistLimitDegrees=45
  twistMotion=Locked
 }
 {
  bone1Name=def_bandolier_front_01
  bone2Name=def_bandolier_front_01_MSTR
  swing1LimitDegrees=5.625
  swing2LimitDegrees=5.625
  twistLimitDegrees=0
 }
 {bone1Name=def_bandolier_front_02, bone2Name=def_bandolier_front_01, swing1LimitDegrees=5.625, swing2LimitDegrees=5.625, twistLimitDegrees=0}
 {bone1Name=def_bandolier_front_03, bone2Name=def_bandolier_front_02, swing1LimitDegrees=5.625, swing2LimitDegrees=5.625, twistLimitDegrees=0}
 {bone1Name=def_bandolier_front_04, bone2Name=def_bandolier_front_03, swing1LimitDegrees=5.625, swing2LimitDegrees=5.625, twistLimitDegrees=0}
 {
  bone1Name=def_bandolier_front_05
  bone2Name=def_bandolier_front_04
  swing1LimitDegrees=0
  swing1Motion=Locked
  swing2LimitDegrees=5.625
  swing2Motion=Locked
  twistLimitDegrees=0
  twistMotion=Locked
 }
 {
  bone1Name=def_bandolier_rear_01_MSTR
  bone2Name=spine_05
  swing1LimitDegrees=45
  swing1Motion=Locked
  swing2LimitDegrees=45
  swing2Motion=Locked
  twistLimitDegrees=45
  twistMotion=Locked
 }
 {
  bone1Name=def_bandolier_rear_01
  bone2Name=def_bandolier_rear_01_MSTR
  swing1LimitDegrees=5.625
  swing2LimitDegrees=5.625
  twistLimitDegrees=0
 }
 {bone1Name=def_bandolier_rear_02, bone2Name=def_bandolier_rear_01, swing1LimitDegrees=5.625, swing2LimitDegrees=5.625, twistLimitDegrees=0}
 {bone1Name=def_bandolier_rear_03, bone2Name=def_bandolier_rear_02, swing1LimitDegrees=5.625, swing2LimitDegrees=5.625, twistLimitDegrees=0}
 {bone1Name=def_bandolier_rear_04, bone2Name=def_bandolier_rear_03, swing1LimitDegrees=5.625, swing2LimitDegrees=5.625, twistLimitDegrees=0}
 {
  bone1Name=def_bandolier_rear_05
  bone2Name=def_bandolier_rear_04
  swing1LimitDegrees=0
  swing1Motion=Locked
  swing2LimitDegrees=5.625
  swing2Motion=Locked
  twistLimitDegrees=0
  twistMotion=Locked
 }
 {
  bone1Name=def_bandolier_rear_05
  bone2Name=dyn_bag_sim_chain_05
  swing1LimitDegrees=45
  swing1Motion=Locked
  swing2LimitDegrees=45
  swing2Motion=Locked
  twistLimitDegrees=45
  twistMotion=Locked
 }
 {
  bone1Name=def_bandolier_front_05
  bone2Name=dyn_bag_sim_chain_05
  swing1LimitDegrees=45
  swing1Motion=Locked
  swing2LimitDegrees=45
  swing2Motion=Locked
  twistLimitDegrees=45
  twistMotion=Locked
 }
]
