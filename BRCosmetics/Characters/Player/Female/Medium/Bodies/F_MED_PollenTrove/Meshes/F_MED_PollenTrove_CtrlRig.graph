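# Control Rig for the PollenTrove female medium body cosmetic that applies pose-driven corrective offsets to its buttock bones.
# From BeginExecution, SphericalPoseReaders on each thigh output weights to ModifyTransforms nodes that translate the matching buttock bone, with mirrored reader settings and differing offsets per side.
# Left-thigh-driven skirt bone corrections and an unused right-thigh reader also exist, but are not linked to execution.
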ControlRigBlueprint /BRCosmetics/Characters/Player/Female/Medium/Bodies/F_MED_PollenTrove/Meshes/F_MED_PollenTrove_CtrlRig
graphs [
 ConstructionGraph {nodes [RigUnit_PrepareForExecution = &RigUnit_PrepareForExecution()]}
 RigVMFunctionLibrary {links [], nodes []}
 RigVMModel {nodes [
   n1 = &RigUnit_SphericalPoseReader(
    ActiveRegionScaleFactors=(PositiveWidth=1,NegativeWidth=1,PositiveHeight=1,NegativeHeight=1)
    ActiveRegionSize=0
    Debug=(bDrawDebug=true,bDraw2D=False,bDrawLocalAxes=False,DebugScale=25,DebugSegments=20,DebugThickness=0.25)
    DriverAxis=(X=-1,Y=0,Z=0)
    DriverItem=(Type=Bone,Name="thigh_l")
    ExecutePin=()
    FalloffRegionScaleFactors=(PositiveWidth=1,NegativeWidth=1,PositiveHeight=1,NegativeHeight=1)
    FalloffSize=0.425
    FlipHeightScaling=False
    FlipWidthScaling=False
    OptionalParentItem=(Type=Bone,Name="None")
    RotationOffset=(X=90,Y=0,Z=0)
   )
   &RigUnit_ModifyTransforms(
    ExecutePin=()
    ItemToModify=((Item=(Type=Bone,Name="dyn_skirt_fr_l_MSTR"),Transform=(Rotation=(X=0,Y=-0.3827,Z=0,W=0.9239),Translation=(X=-10,Y=0,Z=0),Scale3D=(X=1,Y=1,Z=1))))
    Weight=@n1.OutputParam
   )
   > &RigUnit_ModifyTransforms(
    ItemToModify=((Item=(Type=Bone,Name="dyn_skirt_fr_l_b"),Transform=(Rotation=(X=0,Y=-0.2588,Z=0,W=0.9659),Translation=(X=0,Y=0,Z=0),Scale3D=(X=1,Y=1,Z=1))))
    Weight=@n1.OutputParam
   )
   &RigUnit_ModifyTransforms(
    ExecutePin=()
    ItemToModify=((Item=(Type=Bone,Name="dyn_skirt_fr_ct_MSTR"),Transform=(Rotation=(X=0,Y=-0.5373,Z=0,W=0.8434),Translation=(X=-3.1318,Y=0.0002,Z=2.3733),Scale3D=(X=1,Y=1,Z=1))))
    Weight=@n1.OutputParam
   )
   > &RigUnit_ModifyTransforms(
    ItemToModify=((Item=(Type=Bone,Name="dyn_skirt_l_MSTR"),Transform=(Rotation=(X=0,Y=0,Z=0.2164,W=0.9763),Translation=(X=0,Y=0,Z=0),Scale3D=(X=1,Y=1,Z=1))))
    Weight=@n1.OutputParam
   )
   RigUnit_BeginExecution = &RigUnit_BeginExecution()
   n2 = &RigUnit_SphericalPoseReader(
    ActiveRegionScaleFactors=(PositiveWidth=4,NegativeWidth=3,PositiveHeight=2.5,NegativeHeight=2)
    ActiveRegionSize=0.1
    Debug=(bDrawDebug=true,bDraw2D=False,bDrawLocalAxes=False,DebugScale=25,DebugSegments=20,DebugThickness=0.25)
    DriverAxis=(X=-1,Y=0,Z=0)
    DriverItem=(Type=Bone,Name="thigh_l")
    FalloffRegionScaleFactors=(PositiveWidth=1,NegativeWidth=4,PositiveHeight=1,NegativeHeight=0.5)
    FalloffSize=0.4
    FlipHeightScaling=False
    FlipWidthScaling=False
    OptionalParentItem=(Type=Bone,Name="None")
    RotationOffset=(X=100,Y=0,Z=-20)
    ExecutePin=@RigUnit_BeginExecution.ExecutePin
   )
   > &RigUnit_ModifyTransforms(
    ItemToModify=((Item=(Type=Bone,Name="buttock_l"),Transform=(Rotation=(X=0,Y=0,Z=0,W=1),Translation=(X=10,Y=-5,Z=0),Scale3D=(X=1,Y=1,Z=1))))
    WeightMinimum=0
    Weight=@n2.OutputParam
   )
   > n3 = &RigUnit_SphericalPoseReader(
    ActiveRegionScaleFactors=(PositiveWidth=4,NegativeWidth=3,PositiveHeight=2.5,NegativeHeight=2)
    ActiveRegionSize=0.1
    Debug=(bDrawDebug=true,bDraw2D=False,bDrawLocalAxes=False,DebugScale=25,DebugSegments=20,DebugThickness=0.25)
    DriverAxis=(X=1,Y=0,Z=0)
    DriverItem=(Type=Bone,Name="thigh_r")
    FalloffRegionScaleFactors=(PositiveWidth=1,NegativeWidth=4,PositiveHeight=1,NegativeHeight=0.5)
    FalloffSize=0.4
    FlipHeightScaling=False
    FlipWidthScaling=true
    OptionalParentItem=(Type=Bone,Name="None")
    RotationOffset=(X=-80,Y=0,Z=-20)
   )
   > &RigUnit_ModifyTransforms(
    ItemToModify=((Item=(Type=Bone,Name="buttock_r"),Transform=(Rotation=(X=0,Y=0,Z=0,W=1),Translation=(X=-7,Y=3,Z=0),Scale3D=(X=1,Y=1,Z=1))))
    WeightMinimum=0
    Weight=@n3.OutputParam
   )
   &RigUnit_SphericalPoseReader(
    ActiveRegionScaleFactors=(PositiveWidth=1,NegativeWidth=1,PositiveHeight=1,NegativeHeight=1)
    ActiveRegionSize=0
    Debug=(bDrawDebug=true,bDraw2D=False,bDrawLocalAxes=False,DebugScale=25,DebugSegments=20,DebugThickness=0.25)
    DriverAxis=(X=-1,Y=0,Z=0)
    DriverItem=(Type=Bone,Name="thigh_r")
    ExecutePin=()
    FalloffRegionScaleFactors=(PositiveWidth=1,NegativeWidth=1,PositiveHeight=1,NegativeHeight=1)
    FalloffSize=0.425
    FlipHeightScaling=False
    FlipWidthScaling=False
    OptionalParentItem=(Type=Bone,Name="None")
    RotationOffset=(X=-90,Y=0,Z=0)
   )
  ]}
]
hierarchy [
 root {loc=<0 0 0>, rot=<0 0 0 1>}
 pelvis {parent=root, loc=<0 -0.5301 98.0571>, rot=<0.0127 0.707 -0.0127 -0.707>}
 spine_01 {parent=pelvis, loc=<6.1208 -0.9328 0>, rot=<0 0 0.034 -0.9994>}
 spine_02 {parent=spine_01, loc=<6.5997 -0.2095 0>, rot=<0 0 0.0238 -0.9997>}
 spine_03 {parent=spine_02, loc=<6.4914 -0.1237 0>, rot=<0 0 0.0149 -0.9999>}
 spine_04 {parent=spine_03, loc=<6.9598 0.0794 0>, rot=<0 0 -0.0132 -0.9999>}
 spine_05 {parent=spine_04, loc=<6.9987 0 0>, rot=<0 0 -0.0015 -1>}
 neck_01 {parent=spine_05, loc=<11.3138 0.1529 0>, rot=<0 0 -0.2003 -0.9797>}
 neck_02 {parent=neck_01, loc=<4.5627 0 0>, rot=<0 0 0.0504 -0.9987>}
 head {parent=neck_02, loc=<5.3914 0 0>, rot=<0 0 0.115 -0.9934>}
 clavicle_l {parent=spine_05, loc=<5.4831 -0.1053 -3.2946>, rot=<0.0465 -0.7289 -0.009 -0.6829>}
 upperarm_l {parent=clavicle_l, loc=<9.8658 -1.2136 -0.4105>, rot=<0.0164 -0.3245 -0.0269 -0.9454>}
 lowerarm_l {parent=upperarm_l, loc=<26.465 -0.8345 0.259>, rot=<0.0036 0.0161 -0.2298 -0.9731>}
 hand_l {parent=lowerarm_l, loc=<22.7355 -0.3203 -0.6105>, rot=<-0.6784 -0.0223 0.0735 -0.7307>}
 hand_attach_l {parent=hand_l, loc=<8.5 -3 0>, rot=<0.7071 0.7071 0 0>}
 weapon_l {parent=hand_l, loc=<-0.1974 -3.4111 0.3811>, rot=<0 0 0 1>}
 thumb_01_l {parent=hand_l, loc=<1.3118 -0.3191 -2.459>, rot=<0.6546 -0.4042 0.0664 -0.6354>}
 thumb_02_l {parent=thumb_01_l, loc=<5.9336 -0.0448 0.0401>, rot=<-0.0037 0.0213 0.298 -0.9543>}
 thumb_03_l {parent=thumb_02_l, loc=<2.2223 0.3682 -0.0907>, rot=<0.0118 -0.013 0.0066 -0.9998>}
 index_metacarpal_l {parent=hand_l, loc=<1.6535 0.0183 -2.96>, rot=<0.0158 -0.0124 -0.0173 -0.9996>}
 index_01_l {parent=index_metacarpal_l, loc=<8.3008 0.0357 -0.0524>, rot=<-0.0044 0.0485 0.0924 -0.9945>}
 index_02_l {parent=index_01_l, loc=<3.3587 -0.3004 0.0452>, rot=<-0.002 0.0027 0.3225 -0.9466>}
 index_03_l {parent=index_02_l, loc=<2.5578 -0.123 -0.0252>, rot=<-0.002 0.0067 0.2558 -0.9667>}
 middle_metacarpal_l {parent=hand_l, loc=<2.1856 -0.1725 -0.5984>, rot=<0.006 0.0168 -0.0329 -0.9993>}
 middle_01_l {parent=middle_metacarpal_l, loc=<7.3275 -0.2447 0.0129>, rot=<-0.0208 0.0042 0.0858 -0.9961>}
 middle_02_l {parent=middle_01_l, loc=<4.0123 0.1401 0.1175>, rot=<-0.0218 0.0258 0.3512 -0.9357>}
 middle_03_l {parent=middle_02_l, loc=<2.6077 0.0534 0.1373>, rot=<0.0082 0.0188 0.1708 -0.9851>}
 ring_metacarpal_l {parent=hand_l, loc=<2.2074 -0.3625 1.3638>, rot=<-0.0284 0.0722 -0.0458 -0.9959>}
 ring_01_l {parent=ring_metacarpal_l, loc=<7.0295 -0.6197 -0.1446>, rot=<-0.0489 -0.0252 0.1892 -0.9804>}
 ring_02_l {parent=ring_01_l, loc=<3.6174 0.4647 -0.1027>, rot=<0 0 0.2806 -0.9598>}
 ring_03_l {parent=ring_02_l, loc=<2.4152 0.114 0.0199>, rot=<0 0 0.1189 -0.9929>}
 pinky_metacarpal_l {parent=hand_l, loc=<1.7616 -0.9753 3.3814>, rot=<-0.0397 0.0814 -0.069 -0.9935>}
 pinky_01_l {parent=pinky_metacarpal_l, loc=<7.1709 -0.4831 -0.1094>, rot=<-0.0654 -0.0035 0.2242 -0.9723>}
 pinky_02_l {parent=pinky_01_l, loc=<2.8274 0.0893 -0.1043>, rot=<0.0037 0.0011 0.2854 -0.9584>}
 pinky_03_l {parent=pinky_02_l, loc=<2.2046 0.0307 -0.0078>, rot=<-0.0002 0.0098 0.1445 -0.9895>}
 lowerarm_twist_02_l {parent=lowerarm_l, loc=<10.6707 0 0>, rot=<0 0 0 1>}
 lowerarm_twist_01_l {parent=lowerarm_l, loc=<18.316 0 0>, rot=<0 0 0 1>}
 upperarm_twist_01_l {parent=upperarm_l, loc=<7.2135 -0.1134 0>, rot=<0 0 0 1>}
 upperarm_twist_02_l {parent=upperarm_l, loc=<19.7998 -0.1134 0>, rot=<0 0 0 1>}
 def_bandolier_root {parent=clavicle_l, loc=<6.328 -0.6686 5.0578>, rot=<0 0 0 1>}
 clavicle_r {parent=spine_05, loc=<5.4831 -0.1053 3.2946>, rot=<0.7289 0.0465 -0.6829 0.009>}
 upperarm_r {parent=clavicle_r, loc=<-9.8658 1.2136 0.4105>, rot=<0.0164 -0.3245 -0.0269 -0.9454>}
 lowerarm_r {parent=upperarm_r, loc=<-26.465 0.8345 -0.259>, rot=<0.0036 0.0161 -0.2298 -0.9731>}
 hand_r {parent=lowerarm_r, loc=<-22.7355 0.3203 0.6106>, rot=<-0.6784 -0.0223 0.0735 -0.7307>}
 thumb_01_r {parent=hand_r, loc=<-1.3118 0.3191 2.459>, rot=<0.6546 -0.4042 0.0664 -0.6354>}
 thumb_02_r {parent=thumb_01_r, loc=<-5.9336 0.0448 -0.0401>, rot=<-0.0037 0.0213 0.298 -0.9543>}
 thumb_03_r {parent=thumb_02_r, loc=<-2.2223 -0.3682 0.0907>, rot=<0.0118 -0.013 0.0066 -0.9998>}
 index_metacarpal_r {parent=hand_r, loc=<-1.6535 -0.0183 2.96>, rot=<0.0158 -0.0124 -0.0173 -0.9996>}
 index_01_r {parent=index_metacarpal_r, loc=<-8.3008 -0.0357 0.0524>, rot=<-0.0044 0.0485 0.0924 -0.9945>}
 index_02_r {parent=index_01_r, loc=<-3.3587 0.3004 -0.0452>, rot=<-0.002 0.0027 0.3225 -0.9466>}
 index_03_r {parent=index_02_r, loc=<-2.5578 0.123 0.0252>, rot=<-0.002 0.0067 0.2558 -0.9667>}
 middle_metacarpal_r {parent=hand_r, loc=<-2.1856 0.1725 0.5984>, rot=<0.006 0.0168 -0.0329 -0.9993>}
 middle_01_r {parent=middle_metacarpal_r, loc=<-7.3275 0.2447 -0.0129>, rot=<-0.0208 0.0042 0.0858 -0.9961>}
 middle_02_r {parent=middle_01_r, loc=<-4.0123 -0.1401 -0.1175>, rot=<-0.0218 0.0258 0.3512 -0.9357>}
 middle_03_r {parent=middle_02_r, loc=<-2.6077 -0.0534 -0.1373>, rot=<0.0082 0.0188 0.1708 -0.9851>}
 ring_metacarpal_r {parent=hand_r, loc=<-2.2074 0.3625 -1.3638>, rot=<-0.0284 0.0722 -0.0458 -0.9959>}
 ring_01_r {parent=ring_metacarpal_r, loc=<-7.0295 0.6197 0.1446>, rot=<-0.0489 -0.0252 0.1892 -0.9804>}
 ring_02_r {parent=ring_01_r, loc=<-3.6174 -0.4647 0.1027>, rot=<0 0 0.2806 -0.9598>}
 ring_03_r {parent=ring_02_r, loc=<-2.4152 -0.114 -0.0199>, rot=<0 0 0.1189 -0.9929>}
 pinky_metacarpal_r {parent=hand_r, loc=<-1.7616 0.9753 -3.3814>, rot=<-0.0397 0.0814 -0.069 -0.9935>}
 pinky_01_r {parent=pinky_metacarpal_r, loc=<-7.1709 0.4831 0.1094>, rot=<-0.0654 -0.0035 0.2242 -0.9723>}
 pinky_02_r {parent=pinky_01_r, loc=<-2.8274 -0.0893 0.1043>, rot=<0.0037 0.0011 0.2854 -0.9584>}
 pinky_03_r {parent=pinky_02_r, loc=<-2.2046 -0.0307 0.0078>, rot=<-0.0002 0.0098 0.1445 -0.9895>}
 weapon_r {parent=hand_r, loc=<0.1974 3.4111 -0.3811>, rot=<0 0 0 1>}
 hand_attach_r {parent=hand_r, loc=<-8.5001 2.9999 0>, rot=<0 0 -0.7071 -0.7071>}
 lowerarm_twist_02_r {parent=lowerarm_r, loc=<-10.6707 0 0>, rot=<0 0 0 -1>}
 lowerarm_twist_01_r {parent=lowerarm_r, loc=<-18.316 0 0>, rot=<0 0 0 -1>}
 upperarm_twist_01_r {parent=upperarm_r, loc=<-7.2135 0.1134 0>, rot=<0 0 0 -1>}
 upperarm_twist_02_r {parent=upperarm_r, loc=<-19.7998 0.1134 0>, rot=<0 0 0 -1>}
 attach_backpack {parent=spine_05, loc=<-0.123 -13.473 0>, rot=<-0.0093 -0.707 0.0093 -0.707>}
 Backpack_BR {parent=attach_backpack, type=Null, loc=<0 0 0>, rot=<0 0 0 1>}
 attach_cape {parent=spine_05, loc=<9.017 -11.714 0>, rot=<0.0616 -0.7044 -0.0616 -0.7044>}
 cape {parent=attach_cape, type=Null, loc=<0 0 0>, rot=<0 0 0 1>}
 dyn_bag_sim_chain_01_MSTR {parent=spine_05, loc=<-5.1845 1.0331 -1.4227>, rot=<0.3076 0.0685 0.9488 -0.0197>}
 dyn_bag_sim_chain_01 {parent=dyn_bag_sim_chain_01_MSTR, loc=<0 0 0>, rot=<0 0 0 1>}
 dyn_bag_sim_chain_02 {parent=dyn_bag_sim_chain_01, loc=<7.9111 0 0>, rot=<0 0.0281 0 -0.9996>}
 dyn_bag_sim_chain_03 {parent=dyn_bag_sim_chain_02, loc=<8.3183 0 0>, rot=<0 0.0099 0 -1>}
 dyn_bag_sim_chain_04 {parent=dyn_bag_sim_chain_03, loc=<7.5473 0 0>, rot=<0 0.0202 0 -0.9998>}
 dyn_bag_sim_chain_05 {parent=dyn_bag_sim_chain_04, loc=<6.8563 0 0>, rot=<0 0 0 1>}
 spline_front_a {parent=dyn_bag_sim_chain_01_MSTR, loc=<-0.855 -11.9018 0.3232>, rot=<0 0 0 1>}
 spline_rear_a {parent=dyn_bag_sim_chain_01_MSTR, loc=<0.838 12.0224 -0.1586>, rot=<0 0 0 1>}
 spline_front_b {parent=dyn_bag_sim_chain_01_MSTR, loc=<7.8683 -12.7022 0.128>, rot=<0 0 0 1>}
 spline_rear_b {parent=dyn_bag_sim_chain_01_MSTR, loc=<7.9492 12.5884 0.0088>, rot=<0 0 0 1>}
 spline_front_c {parent=dyn_bag_sim_chain_01_MSTR, loc=<17.4493 -11.4732 0.5127>, rot=<0 0 0 1>}
 spline_rear_c {parent=dyn_bag_sim_chain_01_MSTR, loc=<15.0655 11.9445 0.3676>, rot=<0 0 0 1>}
 spline_front_d {parent=dyn_bag_sim_chain_01_MSTR, loc=<23.879 -9.7477 1.0734>, rot=<0 0 0 1>}
 spline_front_e {parent=spline_front_d, loc=<6.982 1.801 0>, rot=<0 0 0 1>}
 spline_rear_d {parent=dyn_bag_sim_chain_01_MSTR, loc=<23.6417 9.7843 0.903>, rot=<0 0 0 1>}
 spline_back_e {parent=spline_rear_d, loc=<6.8251 -1.2225 0>, rot=<0 0 0 1>}
 def_bandolier_front_01_MSTR {parent=spine_05, loc=<-5.2513 12.8557 -3.0686>, rot=<0.2897 -0.0147 0.9555 0.0526>}
 def_bandolier_front_01 {parent=def_bandolier_front_01_MSTR, loc=<0 0 0>, rot=<-0.1013 0.0157 -0.0444 -0.9937>}
 def_bandolier_front_02 {parent=def_bandolier_front_01, loc=<8.7232 -0.8004 -0.1952>, rot=<0 0 0 1>}
 def_bandolier_front_03 {parent=def_bandolier_front_02, loc=<9.581 1.2291 0.3847>, rot=<0 0 0 1>}
 def_bandolier_front_04 {parent=def_bandolier_front_03, loc=<6.4298 1.7254 0.5606>, rot=<0 0 0 1>}
 def_bandolier_front_05 {parent=def_bandolier_front_04, loc=<6.982 1.801 0>, rot=<0 0 0 1>}
 def_bandolier_rear_01_MSTR {parent=spine_05, loc=<-4.9986 -10.8856 0.3594>, rot=<0.3223 0.0369 0.9446 -0.0499>}
 def_bandolier_rear_01 {parent=def_bandolier_rear_01_MSTR, loc=<0 0 0>, rot=<-0.0207 -0.0126 0.0394 -0.9989>}
 def_bandolier_rear_02 {parent=def_bandolier_rear_01, loc=<7.1113 0.566 0.1673>, rot=<0 0 0 1>}
 def_bandolier_rear_03 {parent=def_bandolier_rear_02, loc=<7.1163 -0.6439 0.3589>, rot=<0 0 0 1>}
 def_bandolier_rear_04 {parent=def_bandolier_rear_03, loc=<8.5762 -2.1602 0.5353>, rot=<0 0 0 1>}
 def_bandolier_rear_05 {parent=def_bandolier_rear_04, loc=<6.8251 -1.2225 0>, rot=<0 0 0 1>}
 pouch_fr_a_MSTR {parent=spine_04, loc=<1.1864 13.4862 -2.0593>, rot=<-0.201 -0.6563 0.7197 -0.1045>}
 pouch_fr_a {parent=pouch_fr_a_MSTR, loc=<0 0 0>, rot=<0 0 0 1>}
 pouch_fr_b_MSTR {parent=spine_04, loc=<-5.7628 14.1071 2.8807>, rot=<-0.1107 -0.6974 0.6819 -0.1906>}
 pouch_fr_b {parent=pouch_fr_b_MSTR, loc=<0 0 0>, rot=<0 0 0 1>}
 pouch_fr_c_MSTR {parent=spine_04, loc=<-12.3344 13.0217 7.8927>, rot=<-0.052 -0.7273 0.661 -0.177>}
 pouch_fr_c {parent=pouch_fr_c_MSTR, loc=<0 0 0>, rot=<0 0 0 1>}
 thigh_l {parent=pelvis, loc=<-1.0855 -0.2109 -8.7053>, rot=<0.0009 0.0355 0.0155 -0.9993>}
 calf_l {parent=thigh_l, loc=<-42.5284 -0.0019 -0.7297>, rot=<0.0006 -0.0161 -0.0391 -0.9991>}
 foot_l {parent=calf_l, loc=<-46.7401 -2.5014 -0.592>, rot=<0.0427 -0.0238 0.0353 -0.9982>}
 ball_l {parent=foot_l, loc=<-3.5658 10.8066 -0.1544>, rot=<0.0001 -0.0001 -0.7063 -0.7079>}
 calf_twist_02_l {parent=calf_l, loc=<-19.1546 -1.7943 -0.2638>, rot=<0.0237 0.0095 -0.0035 -0.9997>}
 calf_twist_01_l {parent=calf_l, loc=<-35.8483 -1.9197 -0.5794>, rot=<0.0237 0.0095 -0.0035 -0.9997>}
 thigh_twist_01_l {parent=thigh_l, loc=<-27.952 0 0>, rot=<0.0013 0 0 -1>}
 buttock_l {parent=thigh_l, loc=<-4.5 -12 0>, rot=<0 0 0 -1>}
 thigh_r {parent=pelvis, loc=<-1.0855 -0.2109 8.7053>, rot=<0.0355 -0.0009 0.9993 0.0155>}
 calf_r {parent=thigh_r, loc=<42.5284 0.0019 0.7297>, rot=<0.0006 -0.0161 -0.0391 -0.9991>}
 foot_r {parent=calf_r, loc=<46.7401 2.5014 0.592>, rot=<0.0427 -0.0238 0.0353 -0.9982>}
 ball_r {parent=foot_r, loc=<3.5658 -10.8066 0.1544>, rot=<0.0001 -0.0001 -0.7063 -0.7079>}
 calf_twist_02_r {parent=calf_r, loc=<19.1546 1.7943 0.2638>, rot=<0.0237 0.0095 -0.0035 -0.9997>}
 calf_twist_01_r {parent=calf_r, loc=<35.8483 1.9197 0.5794>, rot=<0.0237 0.0095 -0.0035 -0.9997>}
 knee_horn {parent=calf_r, loc=<0.8008 -4.9263 -0.1683>, rot=<0 0 0 -1>}
 thigh_twist_01_r {parent=thigh_r, loc=<27.952 0 0>, rot=<0.0013 0 0 -1>}
 buttock_r {parent=thigh_r, loc=<4.5 12 0>, rot=<0 0 0 -1>}
 dyn_skirt_bounce {parent=pelvis, loc=<0 0 0>, rot=<0 0 0 1>}
 dyn_skirt_ba_ct_MSTR {parent=pelvis, loc=<8.0834 -13.407 -0.5367>, rot=<0.0313 -0.7073 0.7054 -0.0331>}
 dyn_skirt_ba_ct_a {parent=dyn_skirt_ba_ct_MSTR, loc=<0 0 0>, rot=<0 0 0 1>}
 dyn_skirt_ba_ct_b {parent=dyn_skirt_ba_ct_a, loc=<8.3769 0.0126 -0.1336>, rot=<0 0 0 1>}
 dyn_skirt_ba_l_MSTR {parent=pelvis, loc=<8.1962 -12.6338 -7.6191>, rot=<0.0002 -0.8483 0.5255 -0.0645>}
 dyn_skirt_ba_l_a {parent=dyn_skirt_ba_l_MSTR, loc=<0 0 0>, rot=<0 0 0 1>}
 dyn_skirt_ba_l_b {parent=dyn_skirt_ba_l_a, loc=<8.5938 0.0794 -0.0947>, rot=<0 0 0 1>}
 dyn_skirt_ba_r_MSTR {parent=pelvis, loc=<7.9837 -12.4359 6.4735>, rot=<0.0565 -0.5804 0.8123 0.0045>}
 dyn_skirt_ba_r_a {parent=dyn_skirt_ba_r_MSTR, loc=<0 0 0>, rot=<0 0 0 1>}
 dyn_skirt_ba_r_b {parent=dyn_skirt_ba_r_a, loc=<8.2544 -0.041 -0.1781>, rot=<0 0 0 1>}
 dyn_skirt_diag_ba_l_MSTR {parent=pelvis, loc=<8.5086 -9.203 -13.1808>, rot=<-0.0302 -0.9655 0.2291 -0.1203>}
 dyn_skirt_diag_ba_l_a {parent=dyn_skirt_diag_ba_l_MSTR, loc=<0 0 0>, rot=<0 0 0 1>}
 dyn_skirt_diag_ba_l_b {parent=dyn_skirt_diag_ba_l_a, loc=<8.4466 0.0804 -0.1949>, rot=<0 0 0 1>}
 dyn_skirt_diag_ba_r_MSTR {parent=pelvis, loc=<8.1374 -8.9314 12.7391>, rot=<-0.0509 -0.3591 0.9287 0.077>}
 dyn_skirt_diag_ba_r_a {parent=dyn_skirt_diag_ba_r_MSTR, loc=<0 0 0>, rot=<-0.0116 0.1409 0.0062 -0.9899>}
 dyn_skirt_diag_ba_r_b {parent=dyn_skirt_diag_ba_r_a, loc=<8.1608 -0.2959 -0.3815>, rot=<0 0 0 1>}
 dyn_skirt_diag_fr_l_MSTR {parent=pelvis, loc=<8.4673 6.011 -13.4855>, rot=<-0.0554 -0.9281 -0.3611 -0.0717>}
 dyn_skirt_diag_fr_l_a {parent=dyn_skirt_diag_fr_l_MSTR, loc=<0 0 0>, rot=<0 0 0 1>}
 dyn_skirt_diag_fr_l_b {parent=dyn_skirt_diag_fr_l_a, loc=<8.2762 0.0245 -0.0127>, rot=<0 0 0 1>}
 dyn_skirt_diag_fr_r_MSTR {parent=pelvis, loc=<7.8289 5.9836 13.8585>, rot=<0.081 0.4683 0.8783 0.0528>}
 dyn_skirt_diag_fr_r_a {parent=dyn_skirt_diag_fr_r_MSTR, loc=<0 0 0>, rot=<0.0867 0 0 -0.9962>}
 dyn_skirt_diag_fr_r_b {parent=dyn_skirt_diag_fr_r_a, loc=<7.5852 -0.2023 0.2031>, rot=<0 0 0 1>}
 dyn_skirt_fr_ct_MSTR {parent=pelvis, loc=<7.9996 11.3508 -0.3313>, rot=<0.0239 0.6869 0.7257 0.0325>}
 dyn_skirt_fr_ct_a {parent=dyn_skirt_fr_ct_MSTR, loc=<0 0 0>, rot=<0 0 0 1>}
 dyn_skirt_fr_ct_b {parent=dyn_skirt_fr_ct_a, loc=<7.8962 -0.0976 0.2698>, rot=<0 0 0 1>}
 dyn_skirt_fr_l_MSTR {parent=pelvis, loc=<8.2948 9.6786 -8.663>, rot=<-0.047 -0.8248 -0.562 -0.041>}
 dyn_skirt_fr_l_a {parent=dyn_skirt_fr_l_MSTR, loc=<0 0 0>, rot=<0 0 0 1>}
 dyn_skirt_fr_l_b {parent=dyn_skirt_fr_l_a, loc=<8.081 0.0491 0.2108>, rot=<0 0 0 1>}
 dyn_skirt_fr_r_MSTR {parent=pelvis, loc=<7.904 9.7083 8.8129>, rot=<0.0424 0.5081 0.8592 0.0426>}
 dyn_skirt_fr_r_a {parent=dyn_skirt_fr_r_MSTR, loc=<0 0 0>, rot=<0 0 0 1>}
 dyn_skirt_fr_r_b {parent=dyn_skirt_fr_r_a, loc=<7.6145 -0.1769 0.3045>, rot=<0 0 0 1>}
 dyn_skirt_l_MSTR {parent=pelvis, loc=<8.6636 -2.0984 -16.1456>, rot=<-0.0622 -0.9744 -0.1834 -0.1139>}
 dyn_skirt_l_a {parent=dyn_skirt_l_MSTR, loc=<0 0 0>, rot=<0 0 0 1>}
 dyn_skirt_l_b {parent=dyn_skirt_l_a, loc=<8.6614 -0.017 -0.2794>, rot=<0 0 0 1>}
 dyn_skirt_r_MSTR {parent=pelvis, loc=<8.0722 -2.1856 15.8881>, rot=<0.1147 -0.0654 0.9905 0.0377>}
 dyn_skirt_r_a {parent=dyn_skirt_r_MSTR, loc=<0 0 0>, rot=<0 0 0 1>}
 dyn_skirt_r_b {parent=dyn_skirt_r_a, loc=<7.8213 -0.5249 -0.338>, rot=<0 0 0 1>}
]
