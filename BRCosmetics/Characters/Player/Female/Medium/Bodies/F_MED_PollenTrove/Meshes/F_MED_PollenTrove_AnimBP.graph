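# Post-body animation Blueprint for the F_MED_PollenTrove female cosmetic character part, adding secondary motion to its bandolier, pouches and skirt.
# The parent mesh's copied pose feeds a Control Rig, a knee-horn constraint, then RigidBody simulation in spine_05 space.
# Weighted multi-bone constraints then place the bandolier root and three pouches, and a final bone modifier offsets and scales skirt_inner.
# Each update refreshes gravity override via a deprecated function and lowers simulation world alpha in the front end.
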AnimBlueprint /BRCosmetics/Characters/Player/Female/Medium/Bodies/F_MED_PollenTrove/Meshes/F_MED_PollenTrove_AnimBP
target_skeleton=&/F_MED_PollenTrove_Skeleton
graphs [
 AnimGraph {nodes [
   n1 = Control Rig(Source=Copy Pose From Mesh() {bUseAttachedParent=true}, Alpha=0.9) {
    alphaBoolBlend {blendOption=Linear}
    controlRigAssetReference {blueprintRigClass=&/F_MED_PollenTrove_CtrlRig.F_MED_PollenTrove_CtrlRig_C}
   }
   n2 = "Constraint - knee_horn (PP)"!(ComponentPose=Local To Component(LocalPose=@n1), Alpha=1, ConstraintWeights_0=0.4817, ConstraintWeights_1:float=0.5183) -> (Pose) {
    alphaBoolBlend {blendOption=Linear}
    boneToModify {boneName=knee_horn}
    constraintSetup [{targetBone {boneName=thigh_r}}, {targetBone {boneName=calf_r}}]
    constraintWeights [1,1]
   }
   n3 = RigidBody!(
    ComponentPose=@n2
    Alpha
    OverrideWorldGravity:Vector=Get Gravity Override()
    ExternalForce
    SimSpaceSettings:Sim Space Settings Structure=Get Sim Space Settings() -> (Sim Space Settings:Sim Space Settings Structure)
   ) -> (Pose) {
    bOverrideWorldGravity=true
    overridePhysicsAsset=&/F_MED_PollenTrove_Bandolier_B
    simulationSpace=BaseBoneSpace
    alphaBoolBlend {blendOption=Linear}
    baseBoneRef {boneName=spine_05}
    componentAppliedLinearAccClamp=<3000 3000 3000>
    componentLinearAccScale=<0.4 0.3 0.3>
    componentLinearVelScale=<0.4 0.3 0.3>
    simSpaceSettings {
     maxAngularAcceleration=1000
     maxAngularVelocity=3
     maxLinearAcceleration=700
     maxLinearVelocity=500
     velocityScaleZ=0
     worldAlpha=0.333
     externalAngularVelocity {z=0.1}
     externalLinearDragV=<0.2 0.2>
    }
   }
   n4 = "Constraint - def_bandolier_root (PP)"!(ComponentPose=@n3, Alpha=1, ConstraintWeights_0=0.1693, ConstraintWeights_1:float=0.8307) -> (Pose) {
    alphaBoolBlend {blendOption=Linear}
    boneToModify {boneName=def_bandolier_root}
    constraintSetup [{targetBone {boneName=spine_05}}, {targetBone {boneName=clavicle_l}}]
    constraintWeights [1,1]
   }
   n5 = "Constraint - pouch_fr_a (PPPPPPPP)"!(
    ComponentPose=@n4
    Alpha=1
    ConstraintWeights_0=0.1
    ConstraintWeights_1:float=0.0204
    ConstraintWeights_2:float=0.0359
    ConstraintWeights_3:float=0.5084
    ConstraintWeights_4:float=0.0008
    ConstraintWeights_5:float=0.1313
    ConstraintWeights_6:float=0.0254
    ConstraintWeights_7:float=0.1778
   ) -> (Pose) {
    alphaBoolBlend {blendOption=Linear}
    boneToModify {boneName=pouch_fr_a}
    constraintSetup [
     {targetBone {boneName=spine_02}}
     {targetBone {boneName=spine_03}}
     {targetBone {boneName=spine_04}}
     {targetBone {boneName=spine_05}}
     {targetBone {boneName=upperarm_l}}
     {targetBone {boneName=def_bandolier_root}}
     {targetBone {boneName=def_bandolier_front_02}}
     {targetBone {boneName=def_bandolier_front_01}}
    ]
    constraintWeights [1,1,1,1,1,1,1,1]
   }
   n6 = "Constraint - pouch_fr_b (PPPPPPPP)"!(
    ComponentPose=@n5
    Alpha=1
    ConstraintWeights_0=0.0243
    ConstraintWeights_1:float=0.1032
    ConstraintWeights_2:float=0.0341
    ConstraintWeights_3:float=0.0346
    ConstraintWeights_4:float=0.0852
    ConstraintWeights_5:float=0.0925
    ConstraintWeights_6:float=0.424
    ConstraintWeights_7:float=0.2021
   ) -> (Pose) {
    alphaBoolBlend {blendOption=Linear}
    boneToModify {boneName=pouch_fr_b}
    constraintSetup [
     {targetBone {boneName=pelvis}}
     {targetBone {boneName=spine_02}}
     {targetBone {boneName=spine_03}}
     {targetBone {boneName=spine_04}}
     {targetBone {boneName=spine_05}}
     {targetBone {boneName=def_bandolier_front_03}}
     {targetBone {boneName=def_bandolier_front_02}}
     {targetBone {boneName=def_bandolier_front_01}}
    ]
    constraintWeights [1,1,1,1,1,1,1,1]
   }
   n7 = "Constraint - pouch_fr_c (PPPPPPPP)"!(
    ComponentPose=@n6
    Alpha=1
    ConstraintWeights_0=0.0338
    ConstraintWeights_1:float=0.0318
    ConstraintWeights_2:float=0.0096
    ConstraintWeights_3:float=0.0039
    ConstraintWeights_4:float=0.1877
    ConstraintWeights_5:float=0.2912
    ConstraintWeights_6:float=0.3547
    ConstraintWeights_7:float=0.0873
   ) -> (Pose) {
    alphaBoolBlend {blendOption=Linear}
    boneToModify {boneName=pouch_fr_c}
    constraintSetup [
     {targetBone {boneName=pelvis}}
     {targetBone {boneName=spine_02}}
     {targetBone {boneName=spine_03}}
     {targetBone {boneName=spine_04}}
     {targetBone {boneName=dyn_bag_sim_chain_04}}
     {targetBone {boneName=def_bandolier_front_03}}
     {targetBone {boneName=def_bandolier_front_02}}
     {targetBone {boneName=def_bandolier_front_01}}
    ]
    constraintWeights [1,1,1,1,1,1,1,1]
   }
   n8 = "Transform (Modify) Bone - Bone: skirt_inner"(
    ComponentPose=Local To Component(LocalPose=Component To Local(ComponentPose=@n7))
    Alpha=1
    Translation="9,0,0"
    Rotation="0,0,0"
    Scale="1,0.9,0.77"
   ) {
    rotationMode=BMM_Additive
    rotationSpace=BCS_BoneSpace
    scaleMode=BMM_Replace
    scaleSpace=BCS_BoneSpace
    translationMode=BMM_Additive
    translationSpace=BCS_BoneSpace
    alphaBoolBlend {blendOption=Linear}
    boneToModify {boneName=skirt_inner}
    scale {y=0.9, z=0.77}
    translation {x=9}
   }
   Output Pose(Result=Component To Local(ComponentPose=@n8)) {layerGroup=None}
  ]}
 EventGraph {nodes [
   Event BlueprintUpdateAnimation()
   > n1 = Gravity Override State Function DEPRECATED(CharacterPartAnimInstance=Self-Reference(), GameTimeInSeconds=GetGameTimeInSeconds(), GravityOverrideParams=Get Gravity Override Params())
   > Set Gravity Override(Gravity Override=@n1)
   > n2 = Branch(Condition=Get bIsInFrontEnd())
   > Set!(
    execute
    Sim Space Settings_WorldAlpha:float=0.125
    Sim Space Settings_VelocityScaleZ:float=0
    Sim Space Settings_DampingAlpha:float=1
    Sim Space Settings_MaxLinearVelocity:float=500
    Sim Space Settings_MaxAngularVelocity:float=3
    Sim Space Settings_MaxLinearAcceleration:float=700
    Sim Space Settings_MaxAngularAcceleration:float=1000
    Sim Space Settings_ExternalLinearDragV:Vector="0.2,0.2,0"
    Sim Space Settings_ExternalLinearVelocity:Vector="0,0,0"
    Sim Space Settings_ExternalAngularVelocity:Vector="0,0,0.1"
   ) -> (then, Output_Get:Sim Space Settings Structure)
   Set(
    execute=@n2.else
    Sim Space Settings_WorldAlpha=0.333
    Sim Space Settings_VelocityScaleZ=0
    Sim Space Settings_DampingAlpha=1
    Sim Space Settings_MaxLinearVelocity=500
    Sim Space Settings_MaxAngularVelocity=3
    Sim Space Settings_MaxLinearAcceleration=700
    Sim Space Settings_MaxAngularAcceleration=1000
    Sim Space Settings_ExternalLinearDragV="0.2,0.2,0"
    Sim Space Settings_ExternalLinearVelocity="0,0,0"
    Sim Space Settings_ExternalAngularVelocity="0,0,0.1"
   )
  ]}
]
variables [
 Gravity Override {type=Vector}
 Gravity Override Params {type=GravityOverrideParamsStruct}
 Sim Space Settings {
  type=SimSpaceSettings
  default {
   dampingAlpha=1
   maxAngularAcceleration=1000
   maxAngularVelocity=3
   maxLinearAcceleration=700
   maxLinearVelocity=500
   velocityScaleZ=0
   worldAlpha=0.333
   externalAngularVelocity=<0 0 0.1>
   externalLinearDragV=<0.2 0.2 0>
  }
 }
]
