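# Physics asset for the face accessory mesh of the female PollenTrove character, covering its dangling hair, lock and necklaces.
# Capsule chains form five two-segment hair pieces and one five-segment lock, while box chains form two necklaces ending in amulets, each chain constrained back to its MSTR root body.
# Unconstrained head, neck, upper-spine and multi-box glasses bodies supply collision volumes.
# Hair and lock joints share asymmetric swing limits with slight twist; necklace joints have zero twist limit.
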PhysicsAsset /BRCosmetics/Characters/Player/Female/Medium/Bodies/F_MED_PollenTrove/Meshes/Parts/F_MED_PollenTrove_FaceAcc_Physics
bodies {
 dyn_amulet_a [{
   shapeType=Box
   extentX=3
   extentY=1.3434
   extentZ=4
   shapeName=dyn_amulet_a_box
   center=<2.654 0.0731 -0.7763>
   rotation=<rot 0 90.017 -96.0362>
  }]
 dyn_amulet_b [{
   shapeType=Box
   extentX=3
   extentY=1.7461
   extentZ=2.5
   shapeName=dyn_amulet_a_box
   center=<2.6485 0.0731 -0.2456>
   rotation=<rot 0 90.017 -91.0361>
  }]
 dyn_hair_ba_01 [{shapeType=Capsule, length=5, radius=1.625, shapeName=dyn_hair_l_01_capsule, center=<4.88 0 0>, rotation=<rot 90 0 0>}]
 dyn_hair_ba_02 [{shapeType=Capsule, length=2, radius=1.625, shapeName=dyn_hair_l_02_capsule, center=<2.5804 0 0>, rotation=<rot 90 0 0>}]
 dyn_hair_ba_MSTR [{shapeType=Capsule, length=0.505, radius=1.625, shapeName=dyn_hair_l_MSTR_capsule, rotation=<rot 90 0 0>}]
 dyn_hair_ba_l_01 [{shapeType=Capsule, length=6.786, radius=1.625, shapeName=dyn_hair_l_01_capsule, center=<4.88 0 0>, rotation=<rot 90 0 0>}]
 dyn_hair_ba_l_02 [{shapeType=Capsule, length=3, radius=1.625, shapeName=dyn_hair_l_02_capsule, center=<2.5804 0 0>, rotation=<rot 90 0 0>}]
 dyn_hair_ba_l_MSTR [{shapeType=Capsule, length=0.505, radius=1.625, shapeName=dyn_hair_l_MSTR_capsule, rotation=<rot 90 0 0>}]
 dyn_hair_ba_r_01 [{shapeType=Capsule, length=6.786, radius=1.625, shapeName=dyn_hair_l_01_capsule, center=<4.88 0 0>, rotation=<rot 90 0 0>}]
 dyn_hair_ba_r_02 [{shapeType=Capsule, length=3, radius=1.625, shapeName=dyn_hair_l_02_capsule, center=<2.5804 0 0>, rotation=<rot 90 0 0>}]
 dyn_hair_ba_r_MSTR [{shapeType=Capsule, length=0.505, radius=1.625, shapeName=dyn_hair_l_MSTR_capsule, rotation=<rot 90 0 0>}]
 dyn_hair_l_01 [{shapeType=Capsule, length=6.786, radius=1.625, shapeName=dyn_hair_l_01_capsule, center=<4.88 0 0>, rotation=<rot 90 0 0>}]
 dyn_hair_l_02 [{shapeType=Capsule, length=3, radius=1.625, shapeName=dyn_hair_l_02_capsule, center=<2.5804 0 0>, rotation=<rot 90 0 0>}]
 dyn_hair_l_MSTR [{shapeType=Capsule, length=0.505, radius=1.625, shapeName=dyn_hair_l_MSTR_capsule, rotation=<rot 90 0 0>}]
 dyn_hair_r_01 [{shapeType=Capsule, length=6.786, radius=1.625, shapeName=dyn_hair_l_01_capsule, center=<4.88 0 0>, rotation=<rot 90 0 0>}]
 dyn_hair_r_02 [{shapeType=Capsule, length=3, radius=1.625, shapeName=dyn_hair_l_02_capsule, center=<2.5804 0 0>, rotation=<rot 90 0 0>}]
 dyn_hair_r_MSTR [{shapeType=Capsule, length=0.505, radius=1.625, shapeName=dyn_hair_l_MSTR_capsule, rotation=<rot 90 0 0>}]
 dyn_lock_A_01 [{shapeType=Capsule, length=1.4, radius=0.4606, shapeName=dyn_lock_A_01_capsule, center=<1.0979 0 0>, rotation=<rot 90 0 0>}]
 dyn_lock_A_02 [{shapeType=Capsule, length=2.25, radius=0.4611, shapeName=dyn_lock_A_02_capsule, center=<1.4271 0 0>, rotation=<rot 90 0 0>}]
 dyn_lock_A_03 [{shapeType=Capsule, length=2.5, radius=0.505, shapeName=dyn_lock_A_03_capsule, center=<1.5948 0 0>, rotation=<rot 90 0 0>}]
 dyn_lock_A_04 [{shapeType=Capsule, length=2.25, radius=0.405, shapeName=dyn_lock_A_04_capsule, center=<1.6983 0 0>, rotation=<rot 90 0 0>}]
 dyn_lock_A_05 [{shapeType=Capsule, length=2.25, radius=0.4167, shapeName=dyn_lock_A_05_capsule, center=<1.4715 0 0>, rotation=<rot 90 0 0>}]
 dyn_lock_A_MSTR [{shapeType=Capsule, length=0.505, radius=0.505, shapeName=dyn_lock_A_MSTR_capsule, rotation=<rot 90 0 0>}]
 dyn_necklace_a [{
   shapeType=Box
   extentX=3
   extentY=0.9229
   extentZ=6
   shapeName=dyn_necklace_a_box
   center=<3.1132 -0.0512 -0.7939>
   rotation=<rot 4.9999 178.0759 -89.9543>
  }]
 dyn_necklace_a_MSTR [{
   shapeType=Box
   extentX=3
   extentY=3
   extentZ=2.0402
   shapeName=dyn_necklace_a_MSTR_box
   center=<0.0741 0 -0.4204>
   rotation=<rot 0 -90 -10.0002>
  }]
 dyn_necklace_b [{
   shapeType=Box
   extentX=3
   extentY=1.3119
   extentZ=9
   shapeName=dyn_necklace_a_box
   center=<3.148 -0.0527 -0.3665>
   rotation=<rot 0 178.0719 -89.9545>
  }]
 dyn_necklace_b_02 [{
   shapeType=Box
   extentX=3
   extentY=1.3815
   extentZ=6
   shapeName=dyn_necklace_a_box
   center=<3.0962 -0.0509 -0.4103>
   rotation=<rot 9.9998 178.0799 -89.9538>
  }]
 dyn_necklace_b_MSTR [{
   shapeType=Box
   extentX=3
   extentY=3
   extentZ=1.6272
   shapeName=dyn_necklace_a_MSTR_box
   center=<0.0775 0 -0.2891>
   rotation=<rot 0 -90 -15.0002>
  }]
 glasses_fr [
  {
   shapeType=Box
   extentX=5
   extentY=7.1003
   extentZ=7
   shapeName=glasses_fr_box
   center=<-0.6311 -2.4589 -2.3013>
   rotation=<rot 20.0275 -90.0017 -0.0045>
  }
  {
   shapeType=Box
   extentX=5
   extentY=7.1932
   extentZ=9
   shapeName=glasses_fr_box
   center=<-0.6334 -6.1062 -3.8933>
   rotation=<rot 85 -90.0279 0>
  }
  {
   shapeType=Box
   extentX=5
   extentY=7.1003
   extentZ=7
   shapeName=glasses_fr_box
   center=<-0.6311 -2.4506 2.3042>
   rotation=<rot -20.0275 -90.0017 -0.0045>
  }
  {
   shapeType=Box
   extentX=5
   extentY=7.1003
   extentZ=9
   shapeName=glasses_fr_box
   center=<-0.6335 -6.3314 3.5559>
   rotation=<rot -85 -90.0279 0>
  }
  {shapeType=Box, extentX=17.8186, extentY=7.5, extentZ=16.8035, shapeName=glasses_fr_box, center=<4.7244 -12.2548 0>}
 ]
 head [
  {shapeType=Capsule, length=6, radius=5.125, shapeName=head_capsule, center=<4.0749 -0.6406 0.1583>, rotation=<rot 0 104.9997 -90>}
  {shapeType=Capsule, length=5.25, radius=3.75, shapeName=head_capsule, center=<-4.3437 -3.8496 0.3321>}
 ]
 neck_01 [{shapeType=Capsule, length=1.5, radius=4.25, shapeName=neck_01_capsule, center=<0.4288 0.542 0>, rotation=<rot 90 0 0>}]
 neck_02 [{shapeType=Capsule, length=2, radius=4.25, shapeName=neck_02_capsule, center=<1.074 0.6389 0>, rotation=<rot 90 0 0>}]
 spine_05 [{shapeType=Capsule, length=10, radius=10.75, center=<-0.4955 -3.1292 0>, rotation=<rot 0 -90 0>}]
}
constraints [
 {bone1Name=dyn_hair_l_02, bone2Name=dyn_hair_l_01, swing1LimitDegrees=20, swing2LimitDegrees=45, twistLimitDegrees=6.25}
 {bone1Name=dyn_hair_l_01, bone2Name=dyn_hair_l_MSTR, swing1LimitDegrees=20, swing2LimitDegrees=45, twistLimitDegrees=6.25}
 {bone1Name=dyn_hair_r_02, bone2Name=dyn_hair_r_01, swing1LimitDegrees=20, swing2LimitDegrees=45, twistLimitDegrees=6.25}
 {bone1Name=dyn_hair_r_01, bone2Name=dyn_hair_r_MSTR, swing1LimitDegrees=20, swing2LimitDegrees=45, twistLimitDegrees=6.25}
 {bone1Name=dyn_hair_ba_r_02, bone2Name=dyn_hair_ba_r_01, swing1LimitDegrees=20, swing2LimitDegrees=45, twistLimitDegrees=6.25}
 {bone1Name=dyn_hair_ba_r_01, bone2Name=dyn_hair_ba_r_MSTR, swing1LimitDegrees=20, swing2LimitDegrees=45, twistLimitDegrees=6.25}
 {bone1Name=dyn_hair_ba_02, bone2Name=dyn_hair_ba_01, swing1LimitDegrees=20, swing2LimitDegrees=45, twistLimitDegrees=6.25}
 {bone1Name=dyn_hair_ba_01, bone2Name=dyn_hair_ba_MSTR, swing1LimitDegrees=20, swing2LimitDegrees=45, twistLimitDegrees=6.25}
 {bone1Name=dyn_hair_ba_l_02, bone2Name=dyn_hair_ba_l_01, swing1LimitDegrees=20, swing2LimitDegrees=45, twistLimitDegrees=6.25}
 {bone1Name=dyn_hair_ba_l_01, bone2Name=dyn_hair_ba_l_MSTR, swing1LimitDegrees=20, swing2LimitDegrees=45, twistLimitDegrees=6.25}
 {bone1Name=dyn_necklace_a, bone2Name=dyn_necklace_a_MSTR, swing1LimitDegrees=45, swing2LimitDegrees=45, twistLimitDegrees=0}
 {bone1Name=dyn_amulet_a, bone2Name=dyn_necklace_a, swing1LimitDegrees=45, swing2LimitDegrees=45, twistLimitDegrees=0}
 {bone1Name=dyn_necklace_b, bone2Name=dyn_necklace_b_MSTR, swing1LimitDegrees=45, swing2LimitDegrees=45, twistLimitDegrees=0}
 {bone1Name=dyn_necklace_b_02, bone2Name=dyn_necklace_b, swing1LimitDegrees=45, swing2LimitDegrees=45, twistLimitDegrees=0}
 {bone1Name=dyn_amulet_b, bone2Name=dyn_necklace_b_02, swing1LimitDegrees=45, swing2LimitDegrees=45, twistLimitDegrees=0}
 {bone1Name=dyn_lock_A_05, bone2Name=dyn_lock_A_04, swing1LimitDegrees=20, swing2LimitDegrees=45, twistLimitDegrees=6.25}
 {bone1Name=dyn_lock_A_04, bone2Name=dyn_lock_A_03, swing1LimitDegrees=20, swing2LimitDegrees=45, twistLimitDegrees=6.25}
 {bone1Name=dyn_lock_A_03, bone2Name=dyn_lock_A_02, swing1LimitDegrees=20, swing2LimitDegrees=45, twistLimitDegrees=6.25}
 {bone1Name=dyn_lock_A_02, bone2Name=dyn_lock_A_01, swing1LimitDegrees=20, swing2LimitDegrees=45, twistLimitDegrees=6.25}
 {bone1Name=dyn_lock_A_01, bone2Name=dyn_lock_A_MSTR, swing1LimitDegrees=20, swing2LimitDegrees=45, twistLimitDegrees=6.25}
]
