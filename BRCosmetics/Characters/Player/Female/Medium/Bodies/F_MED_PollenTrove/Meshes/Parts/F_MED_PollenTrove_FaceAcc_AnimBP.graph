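# Post-process style AnimBP for the F_MED_PollenTrove face accessory, adding secondary physics to the pose copied from the attached parent mesh.
# A RigidBody node simulates relative to spine_03, then two spring-driven Anim Dynamics nodes swing the left and right earrings.
# Each update computes a gravity override, using a deprecated state function, as the RigidBody's external force and lowers that node's alpha while the character is DBNO.
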AnimBlueprint /BRCosmetics/Characters/Player/Female/Medium/Bodies/F_MED_PollenTrove/Meshes/Parts/F_MED_PollenTrove_FaceAcc_AnimBP
target_skeleton=&/F_MED_PollenTrove_FaceAcc_Skeleton
graphs [
 AnimGraph {nodes [
   n1 = RigidBody(
    ComponentPose=Local To Component(LocalPose=Copy Pose From Mesh() {bUseAttachedParent=true})
    Alpha=Get alphaRBAN() -> (alphaRBAN:double)
    ExternalForce=Get Gravity Override()
   ) {
    bOverrideWorldGravity=true
    overridePhysicsAsset=&/F_MED_PollenTrove_FaceAcc_Physics
    simulationSpace=BaseBoneSpace
    alphaBoolBlend {blendOption=Linear}
    baseBoneRef {boneName=spine_03}
    componentAppliedLinearAccClamp=<3000 3000 3000>
    overrideWorldGravity {z=-980}
    simSpaceSettings {maxAngularAcceleration=20, maxAngularVelocity=4, maxLinearAcceleration=2500, maxLinearVelocity=800, worldAlpha=0.3}
   }
   n2 = Anim Dynamics - Bone: dyn_earring_l(ComponentPose=@n1, Alpha=1, ExternalForce="0,0,0") {
    angularDampingOverride=0.94
    angularSpringConstant=5000
    bAngularSpring=true
    bOverrideAngularDamping=true
    bUseGravityOverride=true
    gravityScale=0.35
    simulationSpace=BoneRelative
    alphaBoolBlend {blendOption=Linear}
    boundBone {boneName=dyn_earring_l}
    componentAppliedLinearAccClamp=<500 500 500>
    componentLinearAccScale=<1 1 1>
    componentLinearVelScale=<1 1 1>
    gravityOverride {z=-490}
    physicsBodyDefinitions [{
      boundBone {boneName=dyn_earring_l}
      boxExtents=<20 20 20>
      constraintSetup {angularLimitsMax=<0 0 25>, angularLimitsMin=<0 -25 -25>, angularTarget=<1 0 0>}
      localJointOffset=<0 0 4>
     }]
    relativeSpaceBone {boneName=spine_03}
   }
   n3 = Anim Dynamics - Bone: dyn_earring_r(ComponentPose=@n2, Alpha=1, ExternalForce="0,0,0") {
    angularDampingOverride=0.94
    angularSpringConstant=5000
    bAngularSpring=true
    bOverrideAngularDamping=true
    bUseGravityOverride=true
    gravityScale=0.35
    simulationSpace=BoneRelative
    alphaBoolBlend {blendOption=Linear}
    boundBone {boneName=dyn_earring_r}
    componentAppliedLinearAccClamp=<500 500 500>
    componentLinearAccScale=<1 1 1>
    componentLinearVelScale=<1 1 1>
    gravityOverride {z=-490}
    physicsBodyDefinitions [{
      boundBone {boneName=dyn_earring_r}
      boxExtents=<20 20 20>
      constraintSetup {angularLimitsMax=<0 0 25>, angularLimitsMin=<0 -25 -25>, angularTarget=<1 0 0>}
      localJointOffset=<0 0 4>
     }]
    relativeSpaceBone {boneName=spine_03}
   }
   Output Pose(Result=Component To Local(ComponentPose=@n3)) {layerGroup=None}
  ]}
 EventGraph {nodes [
   Event BlueprintUpdateAnimation()
   > n1 = Gravity Override State Function DEPRECATED(CharacterPartAnimInstance=Self-Reference(), GameTimeInSeconds=GetGameTimeInSeconds(), GravityOverrideParams=Get Gravity Override Params())
   > Set Gravity Override(Gravity Override=@n1)
   > n2 = Branch(Condition=Get bIsDBNO())
   > Set alphaRBAN(alphaRBAN:double=0.35) -> (Output_Get:double)
   Set alphaRBAN(execute=@n2.else, alphaRBAN=1)
  ]}
]
variables [
 alphaRBAN {type=number}
 Gravity Override {type=Vector}
 Gravity Override Params {type=GravityOverrideParamsStruct, default {dBNO_Carried {jointName=root, gravityOverride=<0 500 -200>}}}
 Override World Gravity {type=Vector, default=<0 0 -980>}
 Sim Space Settings {
  type=SimSpaceSettings
  default {
   dampingAlpha=1
   maxAngularAcceleration=10000
   maxAngularVelocity=8
   maxLinearAcceleration=10000
   maxLinearVelocity=800
   velocityScaleZ=1
   worldAlpha=0.35
   externalLinearDragV=<0 0 0>
  }
 }
]
