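# Control Rig for the head, neck and face-accessory part of the female medium PollenTrove player cosmetic, currently holding no rig logic.
# The graph contains only a Begin Execution node with nothing linked, and the function library is empty.
# The imported hierarchy runs from root through spine, neck and head, with glasses, "dyn_" earring and hair-lock chains under the head and necklace bones under spine_05.
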ControlRigBlueprint /BRCosmetics/Characters/Player/Female/Medium/Bodies/F_MED_PollenTrove/Meshes/Parts/F_MED_PollenTrove_FaceAcc_HeadNeck_CtrlRig
graphs [
 RigVMFunctionLibrary {links [], nodes []}
 RigVMModel {nodes [RigUnit_BeginExecution = &RigUnit_BeginExecution()]}
]
hierarchy [
 root {loc=<0 0 0>, rot=<0 0 0 1>}
 pelvis {parent=root, loc=<0 -0.5301 98.0571>, rot=<0.0127 0.707 -0.0127 -0.707>}
 spine_01 {parent=pelvis, loc=<6.1208 -0.9328 0>, rot=<0 0 0.034 -0.9994>}
 spine_02 {parent=spine_01, loc=<6.5997 -0.2095 0>, rot=<0 0 0.0238 -0.9997>}
 spine_03 {parent=spine_02, loc=<6.4914 -0.1237 0>, rot=<0 0 0.0149 -0.9999>}
 spine_04 {parent=spine_03, loc=<6.9598 0.0794 0>, rot=<0 0 -0.0132 -0.9999>}
 spine_05 {parent=spine_04, loc=<6.9987 0 0>, rot=<0 0 -0.0015 -1>}
 neck_01 {parent=spine_05, loc=<11.3138 0.1529 0>, rot=<0 0 -0.2003 -0.9797>}
 neck_02 {parent=neck_01, loc=<4.5627 0 0>, rot=<0 0 0.0504 -0.9987>}
 head {parent=neck_02, loc=<5.3914 0 0>, rot=<0 0 0.115 -0.9934>}
 glasses_ba {parent=head, loc=<4.1545 -0.3164 0>, rot=<0 0 0 1>}
 glasses_fr {parent=glasses_ba, loc=<-1.0326 9.1397 0>, rot=<0 0 0 1>}
 dyn_earring_l_MSTR {parent=head, loc=<0.545 0.4601 -6.2955>, rot=<0.0044 -0.9688 -0.2478 0.0024>}
 dyn_earring_l {parent=dyn_earring_l_MSTR, loc=<0 0 0>, rot=<0 0 0 1>}
 dyn_earring_r_MSTR {parent=head, loc=<0.5445 0.4533 6.2831>, rot=<0.0001 0.2398 0.9708 -0.005>}
 dyn_earring_r {parent=dyn_earring_r_MSTR, loc=<0 0 0>, rot=<0 0 0 1>}
 dyn_hair_l_MSTR {parent=head, loc=<7.9881 -2.6737 -6.0196>, rot=<0.0728 -0.9944 -0.0162 -0.075>}
 dyn_hair_l_01 {parent=dyn_hair_l_MSTR, loc=<0 0 0>, rot=<0 0 0 1>}
 dyn_lock_B_01 {parent=dyn_hair_l_01, loc=<0.3383 -0.2029 1.5478>, rot=<0.0543 0.1421 -0.0218 -0.9881>}
 dyn_lock_B_02 {parent=dyn_lock_B_01, loc=<2.2315 -0.2716 0.0052>, rot=<-0.0269 0.0488 -0.1512 -0.9869>}
 dyn_lock_B_03 {parent=dyn_lock_B_02, loc=<2.3538 0 0>, rot=<0.4992 0.2699 -0.4844 -0.6658>}
 dyn_lock_C_01 {parent=dyn_hair_l_01, loc=<0.0452 -3.0953 0.5813>, rot=<0.0075 0.2635 -0.0689 -0.9622>}
 dyn_lock_C_02 {parent=dyn_lock_C_01, loc=<3.0597 0 0>, rot=<-0.0207 -0.0133 0.0125 -0.9996>}
 dyn_lock_C_03 {parent=dyn_lock_C_02, loc=<3.1911 0 0>, rot=<0 0 0 1>}
 dyn_hair_l_02 {parent=dyn_hair_l_01, loc=<9.617 0 0>, rot=<0 0 0 1>}
 dyn_hair_ba_l_MSTR {parent=head, loc=<8.0139 -5.3788 -4.3141>, rot=<0.051 -0.8963 0.4289 -0.1006>}
 dyn_hair_ba_l_01 {parent=dyn_hair_ba_l_MSTR, loc=<0 0 0>, rot=<0 0 0 1>}
 dyn_hair_ba_l_02 {parent=dyn_hair_ba_l_01, loc=<9.6485 0 0>, rot=<0 0 0 1>}
 dyn_hair_ba_MSTR {parent=head, loc=<8.0285 -6.8199 0>, rot=<0.0872 -0.7017 0.7017 -0.0872>}
 dyn_hair_ba_01 {parent=dyn_hair_ba_MSTR, loc=<0 0 0>, rot=<0 0 0 1>}
 dyn_lock_D_01 {parent=dyn_hair_ba_01, loc=<4.3873 3.1107 2.7091>, rot=<0.0505 0.1709 0.096 -0.9793>}
 dyn_lock_D_02 {parent=dyn_lock_D_01, loc=<2.824 0 0>, rot=<0.0137 -0.1431 0.1042 -0.9841>}
 dyn_lock_D_03 {parent=dyn_lock_D_02, loc=<2.3507 0 0>, rot=<0 0 0 1>}
 dyn_lock_E_01 {parent=dyn_hair_ba_01, loc=<6.5421 -0.4806 3.8914>, rot=<0.0229 0.021 0.0349 -0.9989>}
 dyn_lock_E_02 {parent=dyn_lock_E_01, loc=<2.6432 0 0>, rot=<-0.0207 -0.0133 0.0125 -0.9996>}
 dyn_hair_ba_02 {parent=dyn_hair_ba_01, loc=<9.6926 0 0>, rot=<0 0 0 1>}
 dyn_hair_ba_r_MSTR {parent=head, loc=<8.015 -5.3911 4.3185>, rot=<0.1149 -0.4697 0.873 -0.0644>}
 dyn_hair_ba_r_01 {parent=dyn_hair_ba_r_MSTR, loc=<0 0 0>, rot=<0 0 0 1>}
 dyn_lock_F_01 {parent=dyn_hair_ba_r_01, loc=<5.1054 2.6727 3.9235>, rot=<0.0503 0.1193 0.1225 -0.984>}
 dyn_lock_F_02 {parent=dyn_lock_F_01, loc=<3.6829 0 0>, rot=<-0.0207 -0.0133 0.0125 -0.9996>}
 dyn_hair_ba_r_02 {parent=dyn_hair_ba_r_01, loc=<9.7312 0 0>, rot=<0 0 0 1>}
 dyn_hair_r_MSTR {parent=head, loc=<7.9789 -1.5588 6.6658>, rot=<0.0759 -0.0064 0.993 -0.0905>}
 dyn_hair_r_01 {parent=dyn_hair_r_MSTR, loc=<0 0 0>, rot=<0 0 0 1>}
 dyn_lock_G_01 {parent=dyn_hair_r_01, loc=<2.2507 3.7815 1.6803>, rot=<0.1283 0.1676 -0.0261 -0.9771>}
 dyn_lock_G_02 {parent=dyn_lock_G_01, loc=<3.2918 0 0>, rot=<-0.0616 -0.1577 0.2171 -0.9614>}
 dyn_lock_G_03 {parent=dyn_lock_G_02, loc=<2.9848 0 0>, rot=<0.1123 -0.2373 0.0057 -0.9649>}
 dyn_lock_G_04 {parent=dyn_lock_G_03, loc=<2.0552 0 0>, rot=<-0.046 0 0 -0.9989>}
 dyn_lock_H_01 {parent=dyn_hair_r_01, loc=<2.5034 -1.8951 2.6573>, rot=<0.0043 0.0293 -0.3296 -0.9436>}
 dyn_lock_H_02 {parent=dyn_lock_H_01, loc=<3.4102 0 0>, rot=<-0.0467 -0.2586 0.3118 -0.9131>}
 dyn_lock_H_03 {parent=dyn_lock_H_02, loc=<2.353 0 0>, rot=<-0.1164 -0.13 0.0603 -0.9828>}
 dyn_hair_r_02 {parent=dyn_hair_r_01, loc=<9.6704 0 0>, rot=<0 0 0 1>}
 dyn_lock_A_MSTR {parent=head, loc=<9.4586 7.5384 -0.4859>, rot=<-0.1666 -0.5528 -0.1187 -0.8078>}
 dyn_lock_A_01 {parent=dyn_lock_A_MSTR, loc=<0 0 0>, rot=<0 0 0 1>}
 dyn_lock_A_02 {parent=dyn_lock_A_01, loc=<2.3288 0 0>, rot=<-0.0287 -0.4626 0.3301 -0.8223>}
 dyn_lock_A_03 {parent=dyn_lock_A_02, loc=<2.9322 0 0>, rot=<-0.1117 -0.1225 0.2014 -0.9654>}
 dyn_lock_A_04 {parent=dyn_lock_A_03, loc=<3.1375 0 0>, rot=<-0.1363 0.0524 0.1484 -0.9781>}
 dyn_lock_A_05 {parent=dyn_lock_A_04, loc=<3.3075 0 0>, rot=<-0.0803 0.1835 0.0881 -0.9758>}
 dyn_necklace_ba_c {parent=spine_05, loc=<12.4369 -3.7842 0.1511>, rot=<0.0284 -0.7065 -0.0284 -0.7065>}
 dyn_necklace_ba_l {parent=spine_05, loc=<12.1946 -2.8086 -3.6119>, rot=<-0.1822 -0.6832 -0.2363 -0.6664>}
 dyn_necklace_side_l {parent=spine_05, loc=<11.327 -0.4889 -5.2616>, rot=<0.4795 0.5196 0.5196 0.4795>}
 dyn_necklace_fr_l {parent=spine_05, loc=<9.9997 2.1344 -5.1229>, rot=<0.4647 0.4333 0.7314 0.2477>}
 dyn_necklace_fr_r {parent=spine_05, loc=<9.8245 2.4881 4.9509>, rot=<-0.2393 -0.7376 -0.4191 -0.4723>}
 dyn_necklace_side_r {parent=spine_05, loc=<11.2997 -0.2603 5.2165>, rot=<0.5196 -0.4795 0.4795 -0.5196>}
 dyn_necklace_ba_r {parent=spine_05, loc=<12.1163 -2.5432 3.9636>, rot=<0.3543 -0.6119 0.3041 -0.6384>}
 dyn_necklace_a_fr_c {parent=spine_05, loc=<7.3989 5.9916 0>, rot=<0.2289 0.669 0.669 0.2289>}
 dyn_necklace_b_fr_c {parent=spine_05, loc=<2.5594 9.2787 0.0006>, rot=<0.2289 0.669 0.669 0.2289>}
 dyn_necklace_b_l_01 {parent=spine_05, loc=<8.765 4.1543 -4.8834>, rot=<0.3814 0.4991 0.7416 0.2355>}
 dyn_necklace_b_l_02 {parent=dyn_necklace_b_l_01, loc=<3.2703 0 0>, rot=<-0.0587 -0.1226 -0.1466 -0.9798>}
 dyn_necklace_b_l_03 {parent=dyn_necklace_b_l_02, loc=<3.4395 0 0>, rot=<0.0757 0.0546 -0.0715 -0.9931>}
 dyn_necklace_b_r_01 {parent=spine_05, loc=<8.0089 4.6878 4.4409>, rot=<-0.2436 -0.7335 -0.5071 -0.3814>}
 dyn_necklace_b_r_02 {parent=dyn_necklace_b_r_01, loc=<2.2515 0 0>, rot=<-0.0579 -0.0947 0.1415 -0.9837>}
 dyn_necklace_b_r_03 {parent=dyn_necklace_b_r_02, loc=<2.812 0 0>, rot=<0.0227 0.0583 -0.0685 -0.9957>}
 dyn_necklace_a_r_01 {parent=spine_05, loc=<8.8238 4.1938 3.8208>, rot=<-0.126 -0.7851 -0.2982 -0.528>}
 dyn_necklace_a_r_02 {parent=dyn_necklace_a_r_01, loc=<2.5 0 0>, rot=<-0.0475 0.0993 0.1675 -0.9797>}
 dyn_necklace_a_l_01 {parent=spine_05, loc=<8.894 4.034 -3.9679>, rot=<0.5311 0.3619 0.7586 0.1074>}
 dyn_necklace_a_l_02 {parent=dyn_necklace_a_l_01, loc=<2.4109 0 0>, rot=<0.0522 0.0909 -0.179 -0.9783>}
]
